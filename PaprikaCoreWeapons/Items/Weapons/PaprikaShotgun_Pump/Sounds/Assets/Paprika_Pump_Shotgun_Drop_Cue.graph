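# Sound cue for the drop sound of the Paprika pump shotgun, using shared Fort weapon-drop attenuation and the loot-collect sound class.
# A Random node picks between two equally weighted drop waves and feeds a root Modulator that slightly varies pitch and volume.
# The cue plays at a reduced volume multiplier and references the Weapon_Drop_StopOldest concurrency set to limit overlapping drops.
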SoundCue /PaprikaCoreWeapons/Items/Weapons/PaprikaShotgun_Pump/Sounds/Assets/Paprika_Pump_Shotgun_Drop_Cue
root=Modulator_0
nodes [
 {id=Modulator_0, children [Random_1], properties {pitchMax=1, pitchMin=0.9, volumeMax=0.9, volumeMin=0.8}}
 {id=Random_1, children [WavePlayer_1,WavePlayer_2], properties {weights [1,1]}}
 {
  id=WavePlayer_1
  properties {soundWaveAssetPtr=&/Paprika_Pump_Shotgun_Drop_01}
 }
 {
  id=WavePlayer_2
  properties {soundWaveAssetPtr=&/Paprika_Pump_Shotgun_Drop_02}
 }
]
properties {
 attenuationSettings=&Fort_Weapon_Drop_Att
 maxDistance=3000
 soundClassObject=&Fort_Loot_Collect
 volumeMultiplier=0.3
 concurrencyOverrides {resolutionRule=PreventNew}
 concurrencySet [&Weapon_Drop_StopOldest]
}
