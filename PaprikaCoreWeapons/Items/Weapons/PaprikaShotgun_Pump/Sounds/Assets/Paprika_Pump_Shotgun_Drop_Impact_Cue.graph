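# Sound cue for the Paprika pump shotgun's drop impact sound.
# A random node picks one of two equally weighted impact waves, then a modulator slightly varies pitch and volume.
# It uses the Fort weapon drop attenuation and Fort loot collect sound class.
# A concurrency override caps simultaneous instances at three and prevents new ones beyond that limit.
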SoundCue /PaprikaCoreWeapons/Items/Weapons/PaprikaShotgun_Pump/Sounds/Assets/Paprika_Pump_Shotgun_Drop_Impact_Cue
root=Modulator_1
nodes [
 {id=Modulator_1, children [Random_1], properties {pitchMax=1, pitchMin=0.9, volumeMax=0.95, volumeMin=0.85}}
 {id=Random_1, children [WavePlayer_0,WavePlayer_1], properties {weights [1,1]}}
 {
  id=WavePlayer_0
  properties {soundWaveAssetPtr=&/Paprika_Pump_Shotgun_DropImpact_01}
 }
 {
  id=WavePlayer_1
  properties {soundWaveAssetPtr=&/Paprika_Pump_Shotgun_DropImpact_02}
 }
]
properties {
 attenuationSettings=&Fort_Weapon_Drop_Att
 bOverrideConcurrency=true
 maxDistance=3000
 soundClassObject=&Fort_Loot_Collect
 volumeMultiplier=0.6
 attenuationOverrides {falloffDistance=1000, attenuationShapeExtents {x=0}}
 concurrencyOverrides {maxCount=3, resolutionRule=PreventNew}
}
