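# Sound cue for equipping the Paprika pump shotgun.
# It picks one of three equally weighted equip waves at random and passes it through a modulator for slight volume variation.
# It borrows weapon reload attenuation, the loot collect sound class and a pickup concurrency set.
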SoundCue /PaprikaCoreWeapons/Items/Weapons/PaprikaShotgun_Pump/Sounds/Assets/Paprika_Pump_Shotgun_Equip_Cue
root=Modulator_1
nodes [
 {id=Modulator_1, children [Random_1], properties {volumeMax=1.1, volumeMin=1}}
 {id=Random_1, children [WavePlayer_0,WavePlayer_1,WavePlayer_3], properties {weights [1,1,1]}}
 {
  id=WavePlayer_0
  properties {soundWaveAssetPtr=&/Paprika_Pump_Shotgun_Equip_01}
 }
 {
  id=WavePlayer_1
  properties {soundWaveAssetPtr=&/Paprika_Pump_Shotgun_Equip_02}
 }
 {
  id=WavePlayer_3
  properties {soundWaveAssetPtr=&/Paprika_Pump_Shotgun_Equip_03}
 }
]
properties {
 attenuationSettings=&Fort_Weapon_Reload_Att
 maxDistance=2250
 soundClassObject=&Fort_Loot_Collect
 volumeMultiplier=1.6
 concurrencyOverrides {maxCount=1, resolutionRule=PreventNew}
 concurrencySet [&PickupMaximum_Oldest]
}
