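# Sound cue for the shell insertion sound of the Paprika pump shotgun, one of its reload foley sounds.
# A Random node picks with equal weight among four InsertShell wave variants to vary repeated inserts.
# It plays quietly through the weapon foley sound class with shared Fortnite weapon reload attenuation, overrides concurrency, and has its debug flag enabled.
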SoundCue /PaprikaCoreWeapons/Items/Weapons/PaprikaShotgun_Pump/Sounds/Assets/Paprika_Pump_Shotgun_InsertShell_Cue
root=Random_0
nodes [
 {id=Random_0, children [WavePlayer_2,WavePlayer_3,WavePlayer_4,WavePlayer_5], properties {weights [1,1,1,1]}}
 {
  id=WavePlayer_2
  properties {soundWaveAssetPtr=&/Paprika_Pump_Shotgun_InsertShell_01}
 }
 {
  id=WavePlayer_3
  properties {soundWaveAssetPtr=&/Paprika_Pump_Shotgun_InsertShell_02}
 }
 {
  id=WavePlayer_4
  properties {soundWaveAssetPtr=&/Paprika_Pump_Shotgun_InsertShell_03}
 }
 {
  id=WavePlayer_5
  properties {soundWaveAssetPtr=&/Paprika_Pump_Shotgun_InsertShell_04}
 }
]
properties {
 attenuationSettings=&Fort_Weapon_Reload_Att
 bDebug=true
 bOverrideConcurrency=true
 maxDistance=2250
 soundClassObject=&Fort_WeaponFoley
 volumeMultiplier=0.2
}
