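# Sound cue for picking up the Paprika pump shotgun.
# A Random node picks between two equally weighted pickup waves, and a Modulator slightly varies the volume on each play.
# It uses weapon pickup attenuation and the loot-collect sound class, with a shared pickup concurrency set plus a one-voice override that prevents new plays.
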SoundCue /PaprikaCoreWeapons/Items/Weapons/PaprikaShotgun_Pump/Sounds/Assets/Paprika_Pump_Shotgun_Pickup_Cue
root=Modulator_1
nodes [
 {id=Modulator_1, children [Random_1], properties {volumeMax=0.9, volumeMin=0.8}}
 {id=Random_1, children [WavePlayer_0,WavePlayer_1], properties {weights [1,1]}}
 {
  id=WavePlayer_0
  properties {soundWaveAssetPtr=&/Paprika_Pump_Shotgun_Pickup_01}
 }
 {
  id=WavePlayer_1
  properties {soundWaveAssetPtr=&/Paprika_Pump_Shotgun_Pickup_02}
 }
]
properties {
 attenuationSettings=&Fort_Weapon_Pickup_Att
 maxDistance=3500
 soundClassObject=&Fort_Loot_Collect
 volumeMultiplier=0.8
 concurrencyOverrides {maxCount=1, resolutionRule=PreventNew}
 concurrencySet [&PickupMaximum_Oldest]
}
